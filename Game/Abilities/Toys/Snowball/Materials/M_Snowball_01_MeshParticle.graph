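# Lit material for Snowball toy ability mesh particles, shading snow with two tintable colors and particle-color-driven emissive.
# Two bump-offset samples of tiled noise at different reference planes fake layered depth.
# Their sum, scaled by tangent-space camera Z, blends the snow colors and normal-map strength; their combined product drives another scalar blend.
# The mobile shading path replaces the blended normal with a constant.
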Material /Game/Abilities/Toys/Snowball/Materials/M_Snowball_01_MeshParticle
graph {
 properties {bUsedWithMeshParticles=true, bUsedWithNiagaraMeshParticles=true, pixelDepthOffsetMode=PDOM_Legacy, usedShadingModels=Default Lit}
 nodes [
  n1 = Transform(None=CameraVectorWS()) {transformSourceType=TRANSFORMSOURCE_World, transformType=TRANSFORM_Tangent}
  n2 = Multiply(A=TextureCoordinate() {uTiling=2, vTiling=2}, B=ScalarParameter() {defaultValue=1, parameterName=Tile})
  n3 = Constant()
  n4 = TextureSample(UVs=BumpOffset(Coordinate=@n2, Height=@n3)) {texture=/Game/Abilities/Toys/Snowball/Textures/T_Snowball_Noise_01}
  n5 = TextureSample(UVs=BumpOffset(Coordinate=@n2, Height=@n3) {referencePlane=0.2}) {texture=&T_Snowball_Noise_01}
  n6 = Add(A=@n4.R, B=@n5.R)
  n7 = Clamp(None=Multiply(A=ComponentMask(None=@n1) {b=true}, B=@n6))
  LinearInterpolate(
   A=VectorParameter() {parameterName=SnowColor_A, defaultValue=<rgba 1 1 1 0>}.RGB
   B=VectorParameter() {parameterName=SnowColor_B, defaultValue=<rgba 0.2989 0.732 0.9896 0>}.RGB
   Alpha=@n7
  )
  LinearInterpolate(Alpha=Multiply(A=@n6, B=Multiply(A=@n4.R, B=@n5.R))) {constA=1, constB=0.2}
  MaterialFunctionCall(Emissive=ParticleColor().RGB) {materialFunction=&MF_FX_EmissiveAdjustments}
  ShadingPathSwitch(
   Default=LinearInterpolate(
    A=Constant3Vector() {constant=<rgba 0 0 1 0>}
    B=TextureSample() {samplerType=SAMPLERTYPE_Normal, texture=/Game/Abilities/Toys/Snowball/Textures/T_Snowball_Normals}.RGB
    Alpha=@n7
   )
   Mobile=Constant3Vector() {constant=<rgba 0.5 0.5 1 0>}
  )
  TextureSample() {texture=&T_Snowball_Noise_01}
 ]
}
